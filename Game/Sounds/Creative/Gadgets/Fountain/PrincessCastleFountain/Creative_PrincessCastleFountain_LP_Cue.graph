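# Sound cue for the looping water ambience of the Creative Princess Castle fountain gadget.
# A Random node with equal weights picks one of four looping wave variations, so each playback settles on a single continuous loop.
# Output uses the ambient water sound class, a fountain-specific attenuation asset and a reduced volume.
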SoundCue /Game/Sounds/Creative/Gadgets/Fountain/PrincessCastleFountain/Creative_PrincessCastleFountain_LP_Cue
root=Random_0
nodes [
 {id=Random_0, children [WavePlayer_0,WavePlayer_1,WavePlayer_2,WavePlayer_3], properties {weights [1,1,1,1]}}
 {
  id=WavePlayer_0
  properties {bLooping=true, soundWaveAssetPtr=&/Creative_PrincessCastleFountain_LP_01}
 }
 {
  id=WavePlayer_1
  properties {bLooping=true, soundWaveAssetPtr=&/Creative_PrincessCastleFountain_LP_02}
 }
 {
  id=WavePlayer_2
  properties {bLooping=true, soundWaveAssetPtr=&/Creative_PrincessCastleFountain_LP_03}
 }
 {
  id=WavePlayer_3
  properties {bLooping=true, soundWaveAssetPtr=&/Creative_PrincessCastleFountain_LP_04}
 }
]
properties {
 attenuationSettings=&/Creative_PrincessCastleFountain_Att
 maxDistance=2800
 soundClassObject=/Game/Sounds/FortSoundClassesAndModes/Fort_Ambient_Water
 volumeMultiplier=0.3
}
